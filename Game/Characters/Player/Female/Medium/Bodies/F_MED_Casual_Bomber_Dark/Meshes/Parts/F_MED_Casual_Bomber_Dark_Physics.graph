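# Physics asset for the female medium Casual Bomber Dark player character part, covering only the dynamic hair, head and upper neck with capsule bodies.
# Two hair chains hang from the head: a four-body dyn_hair_05 chain with moderate swing limits, and a stiffer three-body strand chain.
# Both chains keep twist tightly limited and taper toward the tip, while the head-to-neck_02 joint leaves swing and twist free.
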PhysicsAsset /Game/Characters/Player/Female/Medium/Bodies/F_MED_Casual_Bomber_Dark/Meshes/Parts/F_MED_Casual_Bomber_Dark_Physics
bodies {
 dyn_hair_05 [{shapeType=Capsule, length=2.25, radius=3.5, center=<0.1245 0 0.505>, rotation=<rot 66 -180 -180>}]
 dyn_hair_05_a [{shapeType=Capsule, length=3, radius=2.75, center=<1.5689 0.0001 0.4489>, rotation=<rot 90 0 0>}]
 dyn_hair_05_b [{shapeType=Capsule, length=2.75, radius=2.5, center=<1.6555 0.0001 0.1334>, rotation=<rot 85.3708 0 0>}]
 dyn_hair_05_c [{shapeType=Capsule, length=2.5, radius=2, center=<1.5562 0.0001 0.1335>, rotation=<rot 90 0 0>}]
 dyn_hair_strand_1 [{shapeType=Capsule, length=2, radius=1.8, center=<0.7087 0.1844 -0.0517>, rotation=<rot 85.8236 -0.0001 -16.8443>}]
 dyn_hair_strand_2 [{shapeType=Capsule, length=2, radius=1.8, center=<0.7771 0 -0.0567>, rotation=<rot 85.8248 -0.0001 -0.0001>}]
 dyn_hair_strand_3 [{shapeType=Capsule, length=2, radius=1.5, center=<1.1593 0.0149 -0.0846>, rotation=<rot 85.8262 -0.0001 2.9953>}]
 head [{shapeType=Capsule, length=3, radius=6, center=<-0.2264 3.5991 -0.6819>, rotation=<rot -90 0 0>}]
 neck_02 [{shapeType=Capsule, length=4.5, radius=5, center=<-1.888 3.7263 0>, rotation=<rot 0 -90 -90>}]
}
constraints [
 {bone1Name=dyn_hair_05, bone2Name=head, swing1LimitDegrees=20, swing2LimitDegrees=20, twistLimitDegrees=5}
 {bone1Name=dyn_hair_05_a, bone2Name=dyn_hair_05, swing1LimitDegrees=45, swing2LimitDegrees=45, twistLimitDegrees=5}
 {bone1Name=dyn_hair_05_b, bone2Name=dyn_hair_05_a, swing1LimitDegrees=45, swing2LimitDegrees=45, twistLimitDegrees=5}
 {bone1Name=dyn_hair_05_c, bone2Name=dyn_hair_05_b, swing1LimitDegrees=45, swing2LimitDegrees=45, twistLimitDegrees=5}
 {bone1Name=dyn_hair_strand_1, bone2Name=head, swing1LimitDegrees=4, swing2LimitDegrees=2, twistLimitDegrees=1}
 {bone1Name=dyn_hair_strand_2, bone2Name=dyn_hair_strand_1, swing1LimitDegrees=10, swing2LimitDegrees=10, twistLimitDegrees=5}
 {bone1Name=dyn_hair_strand_3, bone2Name=dyn_hair_strand_2, swing1LimitDegrees=10, swing2LimitDegrees=10, twistLimitDegrees=5}
 {
  bone1Name=head
  bone2Name=neck_02
  swing1LimitDegrees=45
  swing1Motion=Free
  swing2LimitDegrees=45
  swing2Motion=Free
  twistLimitDegrees=45
  twistMotion=Free
 }
]
